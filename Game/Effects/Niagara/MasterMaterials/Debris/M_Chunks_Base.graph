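# Masked, lit base material for rock-chunk debris on Niagara and particle sprites.
# Particle speed crossfades sharp 2x2 chunk alpha and normals into blurred versions to fake motion blur; mobile uses only the sharp maps.
# Particle color drives base color and a fake-lit emissive, whose shadow tint can be shaded by normal versus sun direction.
# A static switch optionally tints both colors by time of day.
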Material /Game/Effects/Niagara/MasterMaterials/Debris/M_Chunks_Base
graph {
 properties {
  bEnableResponsiveAA=true
  bTangentSpaceNormal=false
  bUsedWithNanite=true
  bUsedWithNiagaraSprites=true
  bUsedWithParticleSprites=true
  blendMode=BLEND_Masked
  pixelDepthOffsetMode=PDOM_Legacy
  usedShadingModels=Default Lit
 }
 nodes [
  n1 = MaterialFunctionCall(
   Color_Morning=VectorParameter() {parameterName=Color Morning, defaultValue=<rgba 0.2795 0.1398 0.1407 1>}.RGB
   Color_Day=VectorParameter() {parameterName=Color Day, defaultValue=<rgba 0.6094 0.545 0.5552 1>}.RGB
   Color_Evening=VectorParameter() {parameterName=Color Evening, defaultValue=<rgba 0.2726 0.1761 0.1282 1>}.RGB
   Color_Night=VectorParameter() {parameterName=Color Night, defaultValue=<rgba 0.3438 0.3438 0.3438 1>}.RGB
  ) {materialFunction=&Color_Time_of_Day}
  n2 = ParticleColor()
  n3 = Multiply(A=VectorParameter() {parameterName=ShadowColor, defaultValue=<rgba 0.1 0.1 0.1 1>}.RGB, B=@n2.RGB)
  n4 = Multiply(A=@n3) {constB=0.75}
  n5 = CollectionParameter() {
   collection=&FortniteMaterialParameters
   parameterName=SunlightDirectionalVector
  }
  n6 = Transform(None=TextureSampleParameter2D() {parameterName=Normal, samplerType=SAMPLERTYPE_Normal, texture=/Game/Effects/Fort_Effects/Textures/Rocks/T_Chunks_2x2_N}.RGB)
  n7 = Transform(None=TextureSampleParameter2D() {parameterName=Normal, samplerType=SAMPLERTYPE_Normal, texture=/Game/Effects/Fort_Effects/Textures/Rocks/T_ChunksBlur_2x2_N}.RGB)
  n8 = Clamp(None=Divide(A=ParticleSpeed(), B=ScalarParameter() {defaultValue=60, parameterName=Speed}))
  n9 = DotProduct(
   A=ComponentMask(None=@n5) {b=true, g=true, r=true}
   B=ShadingPathSwitch(Default=LinearInterpolate(A=@n6, B=@n7, Alpha=@n8), Mobile=@n6) {desc=Migrated from FeatureLevelSwitch}
  )
  n10 = Multiply(
   A=LinearInterpolate(A=@n3, B=LinearInterpolate(A=@n3, B=@n4, Alpha=@n9), Alpha=ScalarParameter() {parameterName=Is Translucent})
   B=ScalarParameter() {defaultValue=1, parameterName=Fake Lighting}
  )
  MaterialFunctionCall(Emissive=StaticSwitchParameter(True=Multiply(A=@n1.Result, B=@n10), False=@n10) {parameterName=UseTimeOfDayColor}) {materialFunction=&MF_FX_EmissiveAdjustments}
  ScalarParameter() {defaultValue=0.7, parameterName=Roughness}
  n11 = TextureSampleParameter2D() {parameterName=Color and Alpha, texture=/Game/Effects/Fort_Effects/Textures/Rocks/T_Chunks_2x2_D}
  n12 = StaticSwitchParameter(True=@n11.A, False=@n11.A) {parameterName=UseAlphaForAlpha}
  n13 = TextureSampleParameter2D() {parameterName=Color and Alpha, texture=/Game/Effects/Fort_Effects/Textures/Rocks/T_ChunksBlur_2x2_D}
  n14 = StaticSwitchParameter(True=@n13.A, False=@n13.A) {parameterName=UseAlphaForAlpha}
  ShadingPathSwitch(Default=Multiply(A=@n2.A, B=LinearInterpolate(A=@n12, B=@n14, Alpha=@n8)), Mobile=@n12) {desc=Migrated from FeatureLevelSwitch}
  n15 = Multiply(A=ParticleColor().RGB, B=ScalarParameter() {parameterName=Particle Color As Base Color})
  StaticSwitchParameter(True=Multiply(A=@n1.Result, B=@n15), False=@n15) {parameterName=UseTimeOfDayColor}
 ]
}
