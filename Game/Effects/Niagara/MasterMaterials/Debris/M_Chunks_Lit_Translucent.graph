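# Lit translucent master material for Niagara sprite debris chunks, using 2x2 rock textures with per-vertex directional lighting.
# Optional UseGlow blends two colors by normal-map blue and Fresnel into particle-tinted emissive, and swaps particle-colored concrete base color for a flat tint.
# Normals can blend toward a dot product with a camera-space direction offset by dynamic parameters.
# Opacity multiplies particle alpha by a selectable texture channel.
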Material /Game/Effects/Niagara/MasterMaterials/Debris/M_Chunks_Lit_Translucent
graph {
 properties {
  bEnableResponsiveAA=true
  bUsedWithNiagaraSprites=true
  blendMode=BLEND_Translucent
  pixelDepthOffsetMode=PDOM_Legacy
  refractionMethod=RM_IndexOfRefraction
  translucencyLightingMode=TLM_VolumetricPerVertexDirectional
  usedShadingModels=Default Lit
 }
 nodes [
  Constant() {r=1}
  n1 = TextureSampleParameter2D() {parameterName=Normal, samplerType=SAMPLERTYPE_Normal, texture=/Game/Effects/Fort_Effects/Textures/Rocks/T_Chunks_2x2_N}
  n2 = ScalarParameter() {defaultValue=3, parameterName=FresnelExp}
  n3 = ScalarParameter() {parameterName=FresnelReflectFraction}
  n4 = Saturate(None=Add(
    A=LinearInterpolate(Alpha=@n1.B) {constA=0.5, constB=0}
    B=Fresnel(ExponentIn=@n2, BaseReflectFractionIn=@n3) {baseReflectFraction=0, exponent=1}
   ))
  n5 = Multiply(
   A=LinearInterpolate(
    A=VectorParameter() {group=Color, parameterName=GlowColor2, defaultValue=<rgba 1 0.01 0 0>}.RGB
    B=VectorParameter() {group=Color, parameterName=GlowColor1, defaultValue=<rgba 1 0.2437 0 0>}.RGB
    Alpha=@n4
   )
   B=Add(A=@n4) {constB=0.2}
  )
  n6 = Clamp(None=@n5)
  n7 = ScalarParameter() {defaultValue=1, parameterName=HDRIntensity}
  n8 = ParticleColor()
  n9 = Multiply(
   A=ShadingPathSwitch(
    Default=MaterialFunctionCall(LDR Input=@n6, HDR Tint=@n5, HDR Intensity Multiplier=@n7) {materialFunction=&DeriveHDRfromLDR}.RemappedRange
    Mobile=@n6
   ) {desc=Migrated from FeatureLevelSwitch}
   B=@n8.RGB
  )
  n10 = Constant()
  MaterialFunctionCall(Emissive=Saturate(None=StaticSwitchParameter(True=@n9, False=@n10) {parameterName=UseGlow})) {materialFunction=&MF_FX_EmissiveAdjustments}
  n11 = VectorParameter() {group=Normal, parameterName=UseDotDirection, channelNames {a=1 to use Dot Direction, b="", g="", r=""}, defaultValue=<rgba 0.1 0 -0.1 0>}
  n12 = DotProduct(
   A=Add(
    A=Transform(None=@n11.RGB) {transformSourceType=TRANSFORMSOURCE_Camera, transformType=TRANSFORM_Tangent}
    B=DynamicParameter() {paramNames [x Dot Direction, y Dot Direction, z Dot Direction, Param4]}.RGB
   )
   B=@n1.RGB
  )
  n13 = Constant3Vector() {constant=<rgba 1 1 0.1 0>}
  Multiply(
   A=LinearInterpolate(A=@n1.RGB, B=Multiply(A=@n12, B=@n13), Alpha=@n11.1 to use Dot Direction)
   B=VectorParameter() {parameterName=NormalIntensity, defaultValue=<rgba 2 2 1 1>}.RGB
  )
  Multiply(
   A=@n8.A
   B=StaticComponentMaskParameter(None=TextureSampleParameter2D() {parameterName=Color and Alpha, texture=/Game/Effects/Fort_Effects/Textures/Rocks/T_Chunks_2x2_D}.RGBA) {defaultG=true, group=Opacity, parameterName=OpacityChannel, sortPriority=1}
  )
  StaticSwitchParameter(
   True=VectorParameter() {parameterName=BaseColorTint, defaultValue=<rgba 0.125 0.0951 0.0892 1>}.RGB
   False=Multiply(A=@n8.RGB, B=TextureSample() {texture=/Game/Environments/Apollo/Textures/T_Apollo_Plaza_Concrete_D}.G)
  ) {parameterName=UseGlow}
 ]
}
