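# Translucent, lit sprite material for GPU glass shard debris in Niagara and particle effects, with screen space reflections.
# A random SubUV function picks one cell of a 2x2 shard atlas per particle.
# An inverted dot product of the normal map and a fixed vector fakes facet highlights, blending particle color from dim to very bright through an emissive adjustment function.
# The same term varies opacity, multiplied by the shard mask and particle alpha.
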Material /Game/Effects/Niagara/MasterMaterials/Debris/M_GPU_GlassShard
graph {
 properties {
  bScreenSpaceReflections=true
  bUsedWithNiagaraSprites=true
  bUsedWithParticleSprites=true
  blendMode=BLEND_Translucent
  pixelDepthOffsetMode=PDOM_Legacy
  translucencyLightingMode=TLM_Surface
  translucencyPass=MTP_BeforeDOF
  usedShadingModels=Default Lit
 }
 nodes [
  Constant()
  Constant() {r=1}
  n1 = ParticleColor()
  n2 = Constant() {r=2}
  n3 = MaterialFunctionCall!(UseTexCoord 1, Rows=@n2, Columns=@n2, Particle Random Value) {materialFunction=/Game/Effects/Fort_Effects/Materials/_Functions/MF_RandomSubUV_GPU_Sprites}
  n4 = OneMinus(None=DotProduct(
    A=Constant3Vector() {constant=<rgba 2 0 1 0>}
    B=TextureSample(UVs=@n3.Result) {samplerType=SAMPLERTYPE_Normal, texture=/Game/Effects/Fort_Effects/Textures/Glass/T_Cine_GlassShard}.RGB
   ))
  MaterialFunctionCall(Emissive=LinearInterpolate(A=Multiply(A=@n1.RGB) {constB=1.2}, B=Multiply(A=@n1.RGB) {constB=100}, Alpha=@n4) {constA=1.2, constB=100}) {materialFunction=&MF_FX_EmissiveAdjustments}
  Multiply(
   A=Multiply(
    A=LinearInterpolate(Alpha=@n4) {constA=0.15}
    B=TextureSample(UVs=@n3.Result) {texture=/Game/Effects/Fort_Effects/Textures/Glass/T_Cine_GlassShard_Mask}.R
   )
   B=ParticleColor().A
  )
 ]
}
