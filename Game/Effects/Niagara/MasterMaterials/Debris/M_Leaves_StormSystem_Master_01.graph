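# Master masked, two-sided foliage material for storm system leaf debris particles, usable on Niagara sprites and meshes.
# A leaf atlas, sampled by random column or per-face offset, is tinted by particle color for base, emissive and subsurface color, with optional sun-direction fake shading.
# Switchable world position offset rebuilds quads from UVs with random size, spinning them around one of three random axes by particle distance from object position.
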Material /Game/Effects/Niagara/MasterMaterials/Debris/M_Leaves_StormSystem_Master_01
graph {
 properties {
  bUsedWithNanite=true
  bUsedWithNiagaraMeshParticles=true
  bUsedWithNiagaraSprites=true
  bUsedWithParticleSprites=true
  blendMode=BLEND_Masked
  pixelDepthOffsetMode=PDOM_Legacy
  shadingModel=MSM_TwoSidedFoliage
  translucencyLightingMode=TLM_VolumetricPerVertexDirectional
  twoSided=true
  usedShadingModels=Two Sided Foliage
 }
 nodes [
  Constant()
  n1 = ParticleColor()
  n2 = CollectionParameter() {
   collection=&FortniteMaterialParameters
   parameterName=SunAndMoonModelDirectionalVector
  }
  n3 = ComponentMask(None=@n2) {b=true, g=true, r=true}
  n4 = VertexNormalWS()
  n5 = StaticSwitchParameter(
   True=LinearInterpolate(
    A=ScalarParameter() {defaultValue=1, parameterName="Fake Shading (Dark)"}
    B=ScalarParameter() {defaultValue=1.8, parameterName="Fake Shading (Light)"}
    Alpha=DotProduct(A=@n3, B=@n4)
   )
   False=Constant() {r=1}
  ) {parameterName=UseFakeShading?}
  n6 = TextureCoordinate() {uTiling=0.5}
  n7 = Constant()
  n8 = If(A=TwoSidedSign(), B=@n7, "A > B"=@n7, "A == B"=@n7, "A < B"=Constant() {r=0.5})
  n9 = ScalarParameter() {defaultValue=4, parameterName=U Images}
  n10 = MaterialFunctionCall!(In=Multiply(A=ParticleRandom(), B=@n9)) {materialFunction=/Engine/Functions/Engine_MaterialFunctions02/Utility/Round}
  n11 = Add(A=TextureCoordinate() {uTiling=0.25}, B=AppendVector(A=Divide(A=@n10.Result, B=@n9), B=Constant()))
  n12 = TextureCoordinate()
  n13 = TextureSampleParameter2D(UVs=StaticSwitchParameter(
    True=Panner(Coordinate=@n6, Time=@n8) {speedX=1}
    False=StaticSwitchParameter(True=@n11, False=@n12) {defaultValue=true, parameterName=Use U Random?}
   ) {parameterName=Use 2 Sided Pan?}) {parameterName="BaseColor&&Mask", texture=/Game/Effects/Fort_Effects/World/Enviroment/Falling_Leaf/Textures/T_Leaves_Medium_4x1}
  n14 = Multiply(A=@n13.RGB, B=ScalarParameter() {parameterName=EmissiveBoost})
  MaterialFunctionCall(Emissive=Multiply(A=@n1.RGB, B=Multiply(A=@n5, B=@n14))) {materialFunction=&MF_FX_EmissiveAdjustments}
  Multiply(A=@n13.RGB, B=VectorParameter() {parameterName=Subsurface Color}.RGB)
  Multiply(A=@n1.RGB, B=Multiply(A=@n5, B=@n13.RGB))
  ScalarParameter() {parameterName=Metallic}
  ScalarParameter() {parameterName=Specular}
  ScalarParameter() {parameterName=Roughness}
  n15 = Normalize(VectorInput=MaterialFunctionCall!(
    Random Picker=ParticleRandom()
    Texture Input
    Option 1=VectorParameter() {parameterName=NormalizeRotationAxis_1, defaultValue=<rgba 1 0.3 -0.2 0>}.RGB
    Option 2=VectorParameter() {parameterName=NormalizeRotationAxis_2, defaultValue=<rgba -0.44 1 0.13 0>}.RGB
    Option 3=VectorParameter() {parameterName=NormalizeRotationAxis_3, defaultValue=<rgba 0.06 -0.22 1 0>}.RGB
   ) {materialFunction=/Game/Effects/Fort_Effects/Materials/_Functions/MF_RGB_Random_01}.Vec3 Out Only)
  n16 = ObjectPositionWS()
  n17 = ParticlePositionWS()
  n18 = Multiply(
   A=Divide(A=Distance(A=@n16, B=@n17), B=ScalarParameter() {defaultValue=80, parameterName=Spin Scale})
   B=Clamp(None=ParticleRandom()) {minDefault=0.55}
  )
  n19 = ParticlePositionWS()
  n20 = TextureCoordinate()
  n21 = MaterialFunctionCall(Float2=Multiply(A=ConstantBiasScale(None=@n20) {bias=-0.5, scale=2}, B=ParticleSize())) {materialFunction=&BreakOutFloat2Components}
  n22 = Transform(None=Constant3Vector() {constant=<rgba 1 0 0 0>})
  n23 = ScalarParameter() {defaultValue=20, parameterName=Size X}
  n24 = Multiply(A=Add(A=ParticleRandom()) {constB=-0.5}, B=ScalarParameter() {parameterName=Random Size Scale})
  n25 = Multiply(A=@n21.R, B=Multiply(A=Normalize(VectorInput=@n22), B=Add(A=@n23, B=@n24)))
  n26 = ScalarParameter() {defaultValue=20, parameterName=Size Y}
  n27 = Transform(None=Constant3Vector() {constant=<rgba 0 1 0 0>})
  n28 = Multiply(A=@n21.G, B=Multiply(A=Add(A=@n24, B=@n26), B=Normalize(VectorInput=@n27)))
  n29 = Add(A=Subtract(A=Add(A=@n25, B=@n28), B=WorldPosition().XYZ), B=ParticlePositionWS())
  n30 = Add(A=WorldPosition().XYZ, B=@n29)
  StaticSwitchParameter(True=Constant(), False=Add(A=RotateAboutAxis(NormalizedRotationAxis=@n15, RotationAngle=@n18, PivotPoint=@n19, Position=@n30), B=@n29)) {parameterName=DeactivateWPO?}
 ]
}
