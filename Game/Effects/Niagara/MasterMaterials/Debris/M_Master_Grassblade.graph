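# Translucent, two-sided lit master material for grass blade debris particles on Niagara sprites, also enabled for instanced static meshes.
# World position offset spins blades around the particle position, varying speed by fractional particle height, and adds a near-camera scale offset.
# Particle color is tinted by time of day (curve atlas or four-color function switch), then hue-shifted for color and adjusted emissive.
# Opacity combines grass texture, particle alpha and near-camera fade; dynamic parameters drive hue, emissive and both camera distances.
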Material /Game/Effects/Niagara/MasterMaterials/Debris/M_Master_Grassblade
graph {
 properties {
  bUsedWithInstancedStaticMeshes=true
  bUsedWithNiagaraSprites=true
  blendMode=BLEND_Translucent
  opacityMaskClipValue=0.1
  pixelDepthOffsetMode=PDOM_Legacy
  twoSided=true
  usedShadingModels=Default Lit
 }
 nodes [
  n1 = VectorParameter() {parameterName=RotationAxis}
  n2 = Time()
  n3 = ParticlePositionWS()
  n4 = LinearInterpolate(
   A=ScalarParameter() {defaultValue=1, parameterName=RotationSpeed}
   B=ScalarParameter() {defaultValue=1, parameterName=RotationSpeed_Low}
   Alpha=Frac(None=ComponentMask(None=@n3) {b=true})
  )
  n5 = WorldPosition()
  n6 = ParticlePositionWS()
  n7 = DynamicParameter() {paramNames [HueShift, EmissiveAmount, NearCamFadeDist, NearCamScaleDist]}
  Add(
   A=RotateAboutAxis(
    NormalizedRotationAxis=Transform(None=@n1.RGB) {transformSourceType=TRANSFORMSOURCE_Local}
    RotationAngle=Multiply(A=@n2, B=@n4)
    PivotPoint=ParticlePositionWS()
    Position=WorldPosition().XYZ
   )
   B=LinearInterpolate(
    A=Subtract(A=@n5.XYZ, B=@n6)
    Alpha=MaterialFunctionCall(Distance=@n7.NearCamScaleDist) {materialFunction=&Near_Camera_Fade}.Result
   )
  )
  n8 = ParticleColor()
  n9 = CurveAtlasRowParameter(CurveTime=CollectionParameter() {collection=&FortniteMaterialParameters, parameterName=TimeOfDay_0-1}) {
   atlas=&AthenaCurveAtlas01
   curve=/Game/Athena/LinearColorCurves/Curve_TimeOfDay_S23
   defaultValue=321
   group=TimeOfDay
   parameterName=TOD_Curve
  }
  n10 = MaterialFunctionCall(
   Color_Morning=VectorParameter() {parameterName="Color_Morning (V3)", defaultValue=<rgba 1 1 1 0>}.RGB
   Color_Day=VectorParameter() {parameterName="Color_Day (V3)", defaultValue=<rgba 1 1 1 0>}.RGB
   Color_Evening=VectorParameter() {parameterName="Color_Evening (V3)", defaultValue=<rgba 1 1 1 0>}.RGB
   Color_Night=VectorParameter() {parameterName="Color_Night (V3)", defaultValue=<rgba 1 1 1 0>}.RGB
  ) {materialFunction=&Color_Time_of_Day}
  n11 = MaterialFunctionCall(
   Hue Shift Percentage=DynamicParameter() {paramNames [HueShift, EmissiveAmount, NearCamFadeDist, NearCamScaleDist]}.HueShift
   Texture=StaticSwitchParameter(True=Multiply(A=@n8.RGB, B=@n9), False=Multiply(A=@n8.RGB, B=@n10.Result)) {parameterName=Use ToD Color Curve}
  ) {materialFunction=&HueShift}
  MaterialFunctionCall(Emissive=Multiply(A=DynamicParameter() {paramNames [HueShift, EmissiveAmount, NearCamFadeDist, NearCamScaleDist]}.EmissiveAmount, B=@n11.Result)) {materialFunction=&MF_FX_EmissiveAdjustments}
  n12 = DynamicParameter() {paramNames [HueShift, EmissiveAmount, NearCamFadeDist, NearCamScaleDist]}
  Multiply(
   A=ParticleColor().A
   B=Multiply(
    A=TextureSample() {samplerType=SAMPLERTYPE_Grayscale, texture=/Game/Effects/Fort_Effects/Textures/Environments/Apollo/T_Apollo_Grass01}.G
    B=MaterialFunctionCall(Distance=@n12.NearCamFadeDist) {materialFunction=&Near_Camera_Fade}.Result
   )
  )
  Reroute(None=@n11.Result)
 ]
}
