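# Unlit, translucent, two-sided Niagara master material for rose petal sprite and mesh particles, with a glowing dissolve.
# Particle color tints a channel or RGB from the base texture.
# Dynamic-parameter Dissolve thresholds UV-gradient-weighted, switchable noise, cutting mask opacity and adding a colored edge band to emissive.
# Depth fade softens opacity on default and deferred paths; DissloveOffset_U drives both noise axes, leaving DissloveOffset_V unused.
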Material /Game/Effects/Niagara/MasterMaterials/Debris/M_RosePetals_01
graph {
 properties {
  bAutomaticallySetUsageInEditor=false
  bUsedWithNiagaraMeshParticles=true
  bUsedWithNiagaraSprites=true
  blendMode=BLEND_Translucent
  shadingModel=MSM_Unlit
  twoSided=true
  usedShadingModels=Unlit
 }
 nodes [
  n1 = Time()
  Abs(None=Sine(None=Multiply(A=@n1) {constB=0.1}))
  Constant() {r=0.6}
  n2 = TextureSampleParameter2D() {parameterName=BaseTexture, texture=/Game/Effects/Niagara/MasterMaterials/Debris/Texture/T_RosePetals_01}
  n3 = LinearInterpolate(
   A=ChannelMaskParameter(None=@n2.RGBA) {parameterName=TextureRGBA_Channel}.R
   B=@n2.RGB
   Alpha=ScalarParameter() {parameterName=Use_BaseDefaultRGB}
  )
  n4 = ParticleColor()
  n5 = VectorParameter() {parameterName=LightEdge_Color, defaultValue=<rgba 1 0.1333 0 1>}
  n6 = DynamicParameter() {paramNames [Dissolve, DissloveOffset_U, DissloveOffset_V, Dissolve_UVTiling]}
  n7 = TextureCoordinate() {uTiling=0.3, vTiling=0.3}
  n8 = DynamicParameter() {paramNames [Dissolve, DissloveOffset_U, DissloveOffset_V, Dissolve_UVTiling]}
  n9 = TextureSample(UVs=Add(A=Multiply(A=@n7, B=@n8.Dissolve_UVTiling), B=AppendVector(A=@n8.DissloveOffset_U, B=@n8.DissloveOffset_U))) {samplerType=SAMPLERTYPE_LinearGrayscale, texture=/Engine/Functions/Engine_MaterialFunctions02/ExampleContent/Textures/LowResBlurredNoise}
  n10 = TextureCoordinate()
  n11 = TextureSampleParameter2D(UVs=Add(A=Multiply(A=@n10, B=@n8.Dissolve_UVTiling), B=AppendVector(A=@n8.DissloveOffset_U, B=@n8.DissloveOffset_U))) {
   parameterName=DissloveTexture02
   samplerType=SAMPLERTYPE_LinearGrayscale
   texture=&LowResBlurredNoise
  }
  n12 = Multiply(
   A=ComponentMask(None=TextureCoordinate() {vTiling=0.5}) {g=true}
   B=LinearInterpolate(A=@n9.R, B=@n11.R, Alpha=ScalarParameter() {parameterName=Use_ChangeDissolveTexture})
  )
  n13 = ScalarParameter() {defaultValue=0.1, parameterName=LightEdge_Width}
  n14 = Constant()
  n15 = Constant() {r=1}
  n16 = If(A=@n6.Dissolve, B=Add(A=@n12) {constB=0}, "A > B"=@n14, "A == B"=@n14, "A < B"=@n15)
  n17 = Subtract(A=If(A=@n6.Dissolve, B=Add(A=@n12, B=@n13) {constB=0.15}, "A > B"=@n14, "A == B"=@n14, "A < B"=@n15), B=@n16)
  MaterialFunctionCall(Emissive=Add(A=Multiply(A=@n3, B=@n4.RGB), B=Multiply(A=@n5.RGB, B=@n17))) {materialFunction=&MF_FX_EmissiveAdjustments}
  n18 = ChannelMaskParameter(None=TextureSampleParameter2D() {parameterName=MaskTexture, texture=&T_RosePetals_01}.RGBA) {maskChannel=Green, parameterName=TextureAlpha_Channel, defaultValue=<rgba 0 1 0 0>}
  n19 = Multiply(A=@n4.A, B=Saturate(None=Subtract(A=@n18.R, B=@n16)))
  n20 = DepthFade(Opacity=@n19, FadeDistance=ScalarParameter() {parameterName=DepthFade})
  ShadingPathSwitch(Default=@n20, Deferred=@n20, Forward=@n19, Mobile=@n19)
 ]
}
